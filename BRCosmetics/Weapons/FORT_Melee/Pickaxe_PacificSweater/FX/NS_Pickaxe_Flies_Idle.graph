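# Niagara idle effect for the PacificSweater pickaxe: a burst of small, dark sprite flies that hover around a socket.
# Particles persist past their lifetime, easing toward an offset socket position with inertia while per-axis waveforms, tuned by User.Speed and User.Range, add erratic wandering.
# Sprite alignment blends wander direction with socket-lag direction, the latter gated by a distance check.
# Two sprite renderers draw the particles, alongside computed glow color and size attributes.
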NiagaraSystem /BRCosmetics/Weapons/FORT_Melee/Pickaxe_PacificSweater/FX/NS_Pickaxe_Flies_Idle
dependencies {
 usedDataInterfaces [&NiagaraDataInterfaceCurve, &NiagaraDataInterfaceSocketReader]
 usedDynamicInputs []
 usedModules [
  &self:SystemSpawnScript.SystemScriptSource.SystemScriptGraph.NiagaraNodeAssignment_0.SetVariables_#1
  &SystemState
 ]
}
emitters {Flies {
  emitterUpdateScript [
   &EmitterState(Life Cycle Mode=::NewEnumerator1, Loop Behavior=::NewEnumerator1, Loop Duration=2, UseLoopDelay=false)
   &SpawnBurst_Instantaneous(Spawn Count=10, Use Spawn Probability=false, UseLoopCountLimit=false)
  ]
  particleSpawnScript [
   &InitializeParticle(
    Color=&RandomRangeLinearColor(Minimum=<rgba 0 0 0 1>, Maximum=<rgba 0.02 0.02 0.02 2>)
    Lifetime Max=1
    Lifetime Mode=::NewEnumerator1
    Mass Mode=::NewEnumerator3
    Sprite Size=&RandomRangeVector2D(Minimum=<0.5 1>, Maximum=<0.7 4>)
    Sprite Size Mode=::NewEnumerator4
    UsePositionOffset=false
    Write Material Random=true
   )
   /SetVariables() {isSetParameters=true}
   /SetVariables() {isSetParameters=true}
  ]
  particleUpdateScript [
   &ParticleState(Kill Particles When Lifetime Has Elapsed=false)
   &ScaleSpriteSize(Uniform Curve Sprite Scale=DataInterface() {propertyValues {Curve {keys [
        <at 0 0 RCIM_Cubic>
        {arriveTangent=0.0309, interpMode=RCIM_Cubic, leaveTangent=0.0309, tangentMode=RCTM_User, time=0.05, value=1}
        {arriveTangent=-0.0071, interpMode=RCIM_Cubic, leaveTangent=-0.0071, tangentMode=RCTM_User, time=0.8, value=1}
       ]}}})
   &SocketLocation(
    Orientation Sampling=::NewEnumerator1
    Position Sampling=::NewEnumerator1
    SamplePosOffset=true
    Sampled Position Offset=<0 0 16>
    Socket Reader=$System.Socket Reader
    Socket Sampling Mode=::NewEnumerator3
    VelocitySpeedLimit=false
   )
   /SetVariables(Particles.RandomMovement:Vector3f=&MakeVector(
     X=&Waveform(
      Global Input Position=&RandomRangeFloat(Minimum=0, Maximum=10)
      "Waveform [1]"=::NewEnumerator0
      "[1] Frequency"=&RandomRangeFloat(Minimum=0.5, Maximum=$User.Speed)
      "[1] Amplitude Min/Max"=<-1 1>
      "[1] Scale"=&RandomRangeFloat(
       Minimum=&MakeFloatFromVector2d(Vector 2D=$User.Range)
       Maximum=&MakeFloatFromVector2d(Vector 2D=$User.Range, Channel=::NewEnumerator1)
      )
     )
     Y=&Waveform(
      Global Input Position=&RandomRangeFloat(Minimum=0, Maximum=1)
      "Waveform [1]"=::NewEnumerator1
      "[1] Frequency"=&RandomRangeFloat(Minimum=0.5, Maximum=$User.Speed)
      "[1] Amplitude Min/Max"=<-1 1>
      "[1] Scale"=&RandomRangeFloat(
       Minimum=&MakeFloatFromVector2d(Vector 2D=$User.Range)
       Maximum=&MakeFloatFromVector2d(Vector 2D=$User.Range, Channel=::NewEnumerator1)
      )
     )
     Z=&Waveform(
      Global Input Position=&RandomRangeFloat(Minimum=0, Maximum=10)
      "Waveform [1]"=::NewEnumerator10
      "[1] Frequency"=&Multiply_Float(A=&RandomRangeFloat(Minimum=0.5, Maximum=$User.Speed), B=2)
      "[1] Amplitude Min/Max"=<-1.2 1.2>
      "[1] Scale"=&RandomRangeFloat(
       Minimum=&MakeFloatFromVector2d(Vector 2D=$User.Range)
       Maximum=&MakeFloatFromVector2d(Vector 2D=$User.Range, Channel=::NewEnumerator1)
      )
     )
    )) {isSetParameters=true}
   /SetVariables(Particles.InertiaPosition:NiagaraPosition=&LerpPosition(A=$Particles.InertiaPosition, B=$Output.SocketLocation.SampledPosition, Alpha=0.3)) {isSetParameters=true}
   /SetVariables(Particles.Position=&AddVectorToPosition(Position=$Particles.InertiaPosition, Vector=$Particles.RandomMovement)) {isSetParameters=true}
   /SetVariables(Particles.GlowColor=<rgba 0 0 0 1>, Particles.GlowSize=&Multiply_Vector2D(A=$Particles.SpriteSize, B=<2 2>)) {isSetParameters=true}
   /SetVariables(Particles.SpriteAlignment=&Multiply_VectorByFloat(
     Vector=&Add_Vector(
      A=&MakeCustomVectorFromBool(
       Bool=&SetBoolByFloatComparison(
        A=&DistanceBetweenPositions(A=$Particles.InertiaPosition, B=$Output.SocketLocation.SampledPosition)
        B=5
       )
       True Vector=&NormalizeVector(
        Vector To Normalize=&Subtract_Vector(
         A=Linked() {linkedVariable {name=Particles.InertiaPosition, type=NiagaraPosition}}
         B=Linked() {linkedVariable {name=Output.SocketLocation.SampledPosition, type=NiagaraPosition}}
        )
        Safe Normalize=true
        Safe Fallback Threshold=0
        Safe Fallback Vector=<0 0 0.001>
       )
       False Vector=<0 0 0>
      )
      B=&NormalizeVector(
       Vector To Normalize=&Subtract_Vector(A=$Particles.RandomMovement, B=$Particles.LastFrameVector)
       Safe Normalize=true
       Safe Fallback Threshold=0
       Safe Fallback Vector=<0 0 0.001>
      )
     )
     Float=0.5
    )) {isSetParameters=true}
   /SetVariables(Particles.LastFrameVector:Vector3f=$Particles.RandomMovement) {isSetParameters=true}
  ]
  rendererClasses [&NiagaraSpriteRendererProperties]
  renderers [
   {rendererClass=&NiagaraSpriteRendererProperties}
   {rendererClass=&NiagaraSpriteRendererProperties, rendererIndex=1}
  ]
 }}
summary {
 systemName=NS_Pickaxe_Flies_Idle
 emitters [{emitterName=Flies, rendererClasses [&NiagaraSpriteRendererProperties]}]
 userVariables [
  User.bIsFrontEnd {type=NiagaraBool}
  User.bIsFrontEndPreview {type=NiagaraBool}
  User.Range {type=Vector2f, description="", defaultValue {struct=Vector2f, value=<10 20>}}
  User.Speed {type=NiagaraFloat, defaultValue=2, description=""}
 ]
}
