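# Animation Blueprint for the PacificSweater pickaxe, adding secondary motion to its ribbons and hanging can over the input pose.
# Four spring-driven AnimDynamics ribbon chains feed additive bone rotations, then a chain-to-can simulation relative to the weapon handle.
# In the front end, update logic lowers dynamics alphas and picks one of two rotation offsets by item-preview state.
# Zipline input events set a vector that the AnimGraph never reads.
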AnimBlueprint /BRCosmetics/Weapons/FORT_Melee/Pickaxe_PacificSweater/Meshes/PacificSweater_Axe_AnimBP
target_skeleton=&/PacificSweater_Axe_Skeleton
graphs [
 AnimGraph {nodes [
   AnimGraphNode_SubInput_0 = Input Pose()
   n1 = Get FEDyn_Alpha() -> (FEDyn_Alpha:double)
   n2 = "Anim Dynamics - Chain: dyn_ribbon04_cr -> dyn_ribbon04End_cr"!(
    ComponentPose=Local To Component(LocalPose=@AnimGraphNode_SubInput_0)
    Alpha=@n1
    ComponentLinearAccScale:Vector="1,1,1"
    ComponentLinearVelScale:Vector="1,1,1"
    ExternalForce="0,0,0"
   ) -> (Pose) {
    angularDampingOverride=1
    angularSpringConstant=11000
    bAngularSpring=true
    bChain=true
    bOverrideAngularDamping=true
    bUseGravityOverride=true
    gravityScale=0.35
    simulationSpace=RootRelative
    worldSpaceMinimumScale=0.2
    boundBone {boneName=dyn_ribbon04_cr}
    chainEnd {boneName=dyn_ribbon04End_cr}
    componentAppliedLinearAccClamp=<4000 4000 4000>
    componentLinearAccScale=<1 1 1>
    componentLinearVelScale=<1 1 1>
    gravityOverride {z=-980}
    physicsBodyDefinitions [
     {
      boundBone {boneName=dyn_ribbon04_cr}
      boxExtents=<21 14 14>
      constraintSetup {angularLimitsMax=<15 2.5 10>, angularLimitsMin=<-15 -47 -10>, angularTarget=<1 0 0>}
      localJointOffset=<4 0 0>
     }
     {
      boundBone {boneName=dyn_ribbon04End_cr}
      boxExtents=<20 17 17>
      constraintSetup {angularLimitsMax=<4 5 5>, angularLimitsMin=<-8.5 -30 -5>, angularTarget=<1 0 0>}
      localJointOffset=<4 0 0>
     }
    ]
    simSpaceSettings {maxAngularAcceleration=350, maxAngularVelocity=75, simSpaceAngularAlpha=1}
   }
   n3 = "Anim Dynamics - Chain: dyn_ribbon03_cr -> dyn_ribbon03End_cr"!(ComponentPose=@n2, Alpha=@n1, ComponentLinearAccScale:Vector="1,1,1", ComponentLinearVelScale:Vector="1,1,1", ExternalForce="0,0,0") -> (Pose) {
    angularDampingOverride=1
    angularSpringConstant=11000
    bAngularSpring=true
    bChain=true
    bOverrideAngularDamping=true
    bUseGravityOverride=true
    gravityScale=0.35
    simulationSpace=RootRelative
    worldSpaceMinimumScale=0.2
    boundBone {boneName=dyn_ribbon03_cr}
    chainEnd {boneName=dyn_ribbon03End_cr}
    componentAppliedLinearAccClamp=<4000 4000 4000>
    componentLinearAccScale=<1 1 1>
    componentLinearVelScale=<1 1 1>
    gravityOverride {z=-980}
    physicsBodyDefinitions [
     {
      boundBone {boneName=dyn_ribbon03_cr}
      boxExtents=<21 14 14>
      constraintSetup {angularLimitsMax=<15 2.5 10>, angularLimitsMin=<-15 -35 -10>, angularTarget=<1 0 0>}
      localJointOffset=<4 0 0>
     }
     {
      boundBone {boneName=dyn_ribbon03End_cr}
      boxExtents=<20 17 17>
      constraintSetup {angularLimitsMax=<8.5 2.5 5>, angularLimitsMin=<-8.5 -30 -5>, angularTarget=<1 0 0>}
      localJointOffset=<4 0 0>
     }
    ]
    simSpaceSettings {maxAngularAcceleration=350, maxAngularVelocity=75, simSpaceAngularAlpha=1}
   }
   n4 = "Anim Dynamics - Chain: dyn_ribbon02_cr -> dyn_ribbon02End_cr"!(ComponentPose=@n3, Alpha=@n1, ComponentLinearAccScale:Vector="1,1,1", ComponentLinearVelScale:Vector="1,1,1", ExternalForce="0,0,0") -> (Pose) {
    angularDampingOverride=1
    angularSpringConstant=11000
    bAngularSpring=true
    bChain=true
    bOverrideAngularDamping=true
    bUseGravityOverride=true
    gravityScale=0.35
    simulationSpace=RootRelative
    worldSpaceMinimumScale=0.2
    boundBone {boneName=dyn_ribbon02_cr}
    chainEnd {boneName=dyn_ribbon02End_cr}
    componentAppliedLinearAccClamp=<4000 4000 4000>
    componentLinearAccScale=<1 1 1>
    componentLinearVelScale=<1 1 1>
    gravityOverride {z=-980}
    physicsBodyDefinitions [
     {
      boundBone {boneName=dyn_ribbon02_cr}
      boxExtents=<21 14 14>
      constraintSetup {angularLimitsMax=<15 2.5 10>, angularLimitsMin=<-15 -47 -10>, angularTarget=<1 0 0>}
      localJointOffset=<4 0 0>
     }
     {
      boundBone {boneName=dyn_ribbon02End_cr}
      boxExtents=<20 17 17>
      constraintSetup {angularLimitsMax=<4 30 5>, angularLimitsMin=<-8.5 -30 -5>, angularTarget=<1 0 0>}
      localJointOffset=<4 0 0>
     }
    ]
    simSpaceSettings {maxAngularAcceleration=350, maxAngularVelocity=75, simSpaceAngularAlpha=1}
   }
   n5 = "Anim Dynamics - Chain: dyn_ribbon01_cr -> dyn_ribbon01End_cr"!(ComponentPose=@n4, Alpha=@n1, ComponentLinearAccScale:Vector="1,1,1", ComponentLinearVelScale:Vector="1,1,1", ExternalForce="0,0,0") -> (Pose) {
    angularDampingOverride=1
    angularSpringConstant=11000
    bAngularSpring=true
    bChain=true
    bOverrideAngularDamping=true
    bUseGravityOverride=true
    gravityScale=0.35
    simulationSpace=RootRelative
    worldSpaceMinimumScale=0.2
    boundBone {boneName=dyn_ribbon01_cr}
    chainEnd {boneName=dyn_ribbon01End_cr}
    componentAppliedLinearAccClamp=<4000 4000 4000>
    componentLinearAccScale=<1 1 1>
    componentLinearVelScale=<1 1 1>
    gravityOverride {z=-980}
    physicsBodyDefinitions [
     {
      boundBone {boneName=dyn_ribbon01_cr}
      boxExtents=<21 14 14>
      constraintSetup {angularLimitsMax=<15 2.5 4>, angularLimitsMin=<-15 -47 -4>, angularTarget=<1 0 0>}
      localJointOffset=<4 0 0>
     }
     {
      boundBone {boneName=dyn_ribbon01End_cr}
      boxExtents=<20 17 17>
      constraintSetup {angularLimitsMax=<4 30 5>, angularLimitsMin=<-8.5 -30 -5>, angularTarget=<1 0 0>}
      localJointOffset=<4 0 0>
     }
    ]
    simSpaceSettings {maxAngularAcceleration=350, maxAngularVelocity=75, simSpaceAngularAlpha=1}
   }
   n6 = Get FERot_AlphaInverse() -> (FERot_AlphaInverse:double)
   n7 = "Transform (Modify) Bone - Bone: dyn_can_JNT"(
    ComponentPose="Transform (Modify) Bone - Bone: base_chain_JNT"(ComponentPose=@n5, Alpha=@n6, Translation="0,0,0", Rotation="-55,0,20", Scale="1,1,1") {rotationMode=BMM_Additive, boneToModify {boneName=base_chain_JNT}, rotation {pitch=-55, roll=20}}
    Alpha=@n6
    Translation="0,0,0"
    Rotation="-40,0,0"
    Scale="1,1,1"
   ) {rotationMode=BMM_Additive, boneToModify {boneName=dyn_can_JNT}, rotation {pitch=-40}}
   n8 = Get FERot_Alpha() -> (FERot_Alpha:double)
   n9 = "Anim Dynamics - Chain: dyn_chain_JNT -> dyn_can_JNT"!(
    ComponentPose="Transform (Modify) Bone - Bone: dyn_can_JNT"(
     ComponentPose="Transform (Modify) Bone - Bone: base_chain_JNT"(ComponentPose=@n7, Alpha=@n8, Translation="0,0,0", Rotation="0,0,-30", Scale="1,1,1") {rotationMode=BMM_Additive, boneToModify {boneName=base_chain_JNT}, rotation {roll=-30}}
     Alpha=@n8
     Translation="0,0,0"
     Rotation="-40,0,0"
     Scale="1,1,1"
    ) {rotationMode=BMM_Additive, boneToModify {boneName=dyn_can_JNT}, rotation {pitch=-40}}
    Alpha=Get FEDynCan_Alpha() -> (FEDynCan_Alpha:double)
    ComponentLinearAccScale:Vector="1,1,1"
    ComponentLinearVelScale:Vector="1,1,1"
    ExternalForce="0,0,0"
   ) -> (Pose) {
    angularDampingOverride=1
    angularSpringConstant=11000
    bChain=true
    bOverrideAngularDamping=true
    bUseGravityOverride=true
    gravityScale=0.35
    simulationSpace=BoneRelative
    worldSpaceMinimumScale=0.2
    boundBone {boneName=dyn_chain_JNT}
    chainEnd {boneName=dyn_can_JNT}
    componentAppliedLinearAccClamp=<4000 4000 4000>
    componentLinearAccScale=<1 1 1>
    componentLinearVelScale=<1 1 1>
    gravityOverride {z=-980}
    physicsBodyDefinitions [
     {
      boundBone {boneName=dyn_chain_JNT}
      boxExtents=<22 22 22>
      constraintSetup {angularLimitsMax=<0 45 20>, angularLimitsMin=<0 -65 -20>, angularTarget=<1 0 0>}
      localJointOffset=<4 0 0>
     }
     {
      boundBone {boneName=dyn_can_JNT}
      boxExtents=<13 13 13>
      constraintSetup {angularLimitsMax=<0 30 30>, angularLimitsMin=<0 -30 -20>, angularTarget=<1 0 0>}
      localJointOffset=<4 0 0>
     }
    ]
    relativeSpaceBone {boneName=Melee_TwoHanded_Handle}
    simSpaceSettings {maxAngularAcceleration=80, maxAngularVelocity=90, simSpaceAngularAlpha=1}
   }
   Output Pose(Result=Component To Local(ComponentPose=@n9)) {layerGroup=None}
  ]}
 EventGraph {nodes [
   Event BlueprintUpdateAnimation()
   > n1 = Branch(Condition=Get bIsInFrontEnd())
   > Set FEDyn_Alpha(FEDyn_Alpha:double=0.25) -> (Output_Get:double)
   > Set FEDynCan_Alpha(FEDynCan_Alpha:double=0.6) -> (Output_Get:double)
   > n2 = Branch(Condition=Get bIsPreviewingItemInFrontEnd() -> (bIsPreviewingItemInFrontEnd:bool))
   > n3 = Set FERot_Alpha(FERot_Alpha:double=0) -> (Output_Get:double)
   Set FERot_AlphaInverse(execute=[@n3.then, @n4.then], FERot_AlphaInverse:double=float - float(A=1, B=Get FERot_Alpha())) -> (Output_Get:double)
   n4 = Set FERot_Alpha(execute=@n2.else, FERot_Alpha=1)
   Set FEDyn_Alpha(execute=@n1.else, FEDyn_Alpha=1)
   n5 = Event BlueprintProcessAnimInputEvent()
   n6 = IsNamed(self=@n5.AnimInputEvent, InEventName=aie.Player.Zipline.Entered)
   n7 = OR Boolean(A=Get bIsMoving2D(), B=Get bIsTacticalSprinting())
   n8 = Get bIsJumping()
   > n9 = Branch(Condition=OR Boolean(A=@n6, B=AND Boolean!(A=@n6, B=@n8, C:bool=@n7) -> (ReturnValue)))
   > Set animDyn on zipline(animDyn on zipline:Vector="0, 0, 0") -> (Output_Get:Vector)
   Branch(execute=@n9.else, Condition=IsNamed(self=@n5.AnimInputEvent, InEventName=aie.Player.Zipline.Exited))
   > Set animDyn on zipline(animDyn on zipline="0.4,0.4,0.4")
   TryGetPawnOwner()
  ]}
]
variables [
 animDyn on zipline {type=Vector, default=<0.4 0.4 0.4>}
 Body_Mesh {type=SceneComponent, default=null}
 FEDyn_Alpha {type=number, default=1}
 FEDynCan_Alpha {type=number, default=1}
 FERot_Alpha {type=number}
 FERot_AlphaInverse {type=number}
]
